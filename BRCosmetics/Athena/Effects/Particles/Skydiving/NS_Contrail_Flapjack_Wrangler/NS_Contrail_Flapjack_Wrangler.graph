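# Niagara skydiving contrail for the Flapjack Wrangler cosmetic, trailing syrup and butter effects from the character's skeleton.
# Two Syrup_Ribbons follow skeletal positions, with unlocked particles rising and reacting to owner velocity.
# Butter_Chips, Hand_Gloops and Hand_Splats sample the hand bones, with gloops re-pinned each update to counter-alternated hands.
# Chips blend from owner-local to free simulated positions early in life; front-end modifiers adjust sizes, rates and velocities for menus.
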NiagaraSystem /BRCosmetics/Athena/Effects/Particles/Skydiving/NS_Contrail_Flapjack_Wrangler/NS_Contrail_Flapjack_Wrangler
dependencies {
 usedDataInterfaces [
  &NiagaraDataInterfaceCurve
  &NiagaraDataInterfaceColorCurve
  &NiagaraDataInterfaceVectorCurve
  /Script/Niagara.NiagaraDataInterfaceSimpleCounter
 ]
 usedDynamicInputs []
 usedModules [/Niagara/Modules/System/SystemLifeCycle, &SystemState]
}
emitters {
 Butter_Chips {
  emitterUpdateScript [
   &EmitterState()
   &SpawnRate(Spawn Probability=0.3, SpawnRate=10, Use Spawn Probability=true)
   &SpawnBurst_Instantaneous(Spawn Count=1, Use Spawn Probability=false) {enabled=false}
  ]
  particleSpawnScript [
   /Niagara/Modules/Spawn/Initialization/InitializeParticle(
    Lifetime=2
    Mesh Scale=&Multiply_VectorByFloat(
     Vector=&FN_FrontEndVector3Modifier(Gameplay=<2 2 2>, Front End=<1.5 1.5 1.5>)
     Float=&RandomRangeFloat(Minimum=0.65, Maximum=1)
    )
    Module.Write Color=true
    Module.Write Lifetime=true
    Module.Write Mass=true
    Module.Write Material Random=true
    Module.Write Position=false
    Module.Write Scale=true
    Module.Write SpriteRotation=false
    Module.Write SpriteSize=false
    Position=&SimulationPosition()
    Sprite Size=<5 5>
   )
   &InheritVelocity(Module.Limit Speed=false, Velocity Limit=1, Velocity Scale=<1 1 0>) {enabled=false}
   &SkeletalMeshLocation(
    Bone Sampling Mode=::NewEnumerator0
    Inherit Velocity=false
    MaskSpawnGroup=false
    Module.Override Randomness=false
    Skeletal Mesh=DataInterface() {propertyValues {
      DefaultMesh=null
      ExcludeBoneName=root
      OptionalComponentClass=null
      PreviewMesh=&M_Med_Soldier_04_Preview
      SkinningMode=SkinOnTheFly
      SoftSourceActor=null
      SourceMode=Default
      UvSetIndex=0
      WholeMeshLOD=-1
      bExcludeBone=false
      bReadDeformedGeometry=true
      bRequireCurrentFrameData=true
      ComponentTags []
      FilteredBones [hand_l,hand_r]
      FilteredSockets []
      MeshUserParameter {parameter {name=None, typeDefHandle {registeredTypeIndex=1}, varData []}}
      SamplingRegions []
     }}
   )
   /SetVariables(Particles.LocalPosition:Vector3f=/Niagara/DynamicInputs/Transforms/TransformPosition(Position=Linked() {linkedVariable {name=Particles.Position, type=NiagaraPosition}}, Source Space=::World, Destination Space=::Local)) {isSetParameters=true}
   /Niagara/Modules/Spawn/Orientation/V2/InitialMeshOrientation(
    Orientation Axis:Vector3f=<1 1 1>
    Orientation Vector=&RandomRangeVector(Minimum=<0 0 0>, Maximum=<360 360 360>)
    Use Orientation Vector=true
    Use Rotation Vector=false
   ) {enabled=false}
   /Niagara/Modules/Spawn/Orientation/V3/InitialMeshOrientation(
    Orientation Axis=<1 1 1>
    Orientation Vector=&RandomRangeVector(Minimum=<0 0 0>, Maximum=<360 360 360>)
    Use Orientation Vector=true
    Use Rotation Vector=false
   ) {module=InitialMeshOrientation001}
   &AddVelocity(Velocity=&Multiply_Vector(A=<1 1 1>, B=&FN_FrontEndVector3Modifier(Gameplay=<0 0 0>, Front End=<0 0 250>)))
  ]
  particleUpdateScript [
   &ParticleState()
   &GravityForce() {enabled=false}
   &AccelerationForce(Acceleration=&Multiply_Vector(
     A=<-0.056 -0.05 100>
     B=&VectorFromFloat(Value=&FloatFromCurve(FloatCurve=DataInterface() {propertyValues {Curve {keys [<at 0 0.5 RCIM_Cubic>, {arriveTangent=1.7942, interpMode=RCIM_Cubic, leaveTangent=1.7942, tangentMode=RCTM_User, time=1, value=1}]}}}))
    ))
   &ScaleVelocity(Velocity Scale=&VectorFromCurve(VectorCurve=DataInterface() {propertyValues {XCurve {keys [<at 0 1>,<at 1 0>]}, YCurve {keys [<at 0 1>,<at 1 0>]}, ZCurve {keys [<at 0 1>,<at 1 0>]}}})) {enabled=false}
   &SolveForcesAndVelocity(Clamp Velocity=false, Limit Acceleration=false)
   /Niagara/Modules/Update/Orientation/V2/UpdateMeshOrientation(Rotation Vector:Vector3f=&Multiply_Vector(
     A=&RandomRangeVector(Minimum=<0.5 0.5 0.5>, Maximum=<2 2 2>)
     B=&RandomRangeVector(Minimum=<-1 -1 -1>, Maximum=<1 1 1>)
    )) {enabled=false}
   /Niagara/Modules/Update/Orientation/V3/UpdateMeshOrientation(Rotation Vector=&Multiply_Vector(
     A=&RandomRangeVector(Minimum=<0.5 0.5 0.5>, Maximum=<2 2 2>)
     B=&RandomRangeVector(Minimum=<-1 -1 -1>, Maximum=<1 1 1>)
    )) {module=UpdateMeshOrientation001}
   &ScaleMeshSize(Scale Factor=&Multiply_VectorByFloat(
     Vector=<1 1 1>
     Float=&FloatFromCurve(FloatCurve=DataInterface() {propertyValues {Curve {keys [
          {arriveTangent=2.2113, interpMode=RCIM_Cubic, leaveTangent=2.2113, tangentMode=RCTM_User, time=0, value=1}
          {arriveTangent=-0.0072, interpMode=RCIM_Cubic, leaveTangent=-0.0072, tangentMode=RCTM_User, time=0.1609, value=2.5058}
          <at 0.2553 1.0392>
          <at 1 0 RCIM_Cubic>
         ]}}})
    ))
   /SetVariables(Particles.Position=&Lerp_Vector(
     A=&TransformPosition(Position=$Particles.LocalPosition, Source Space=::Local, Destination Space=::Simulation)
     B=Linked() {linkedVariable {name=Particles.Position, type=NiagaraPosition}}
     Alpha=&FloatFromCurve(FloatCurve=DataInterface() {propertyValues {Curve {keys [<at 0 0 RCIM_Cubic>, {arriveTangent=4.4737, interpMode=RCIM_Cubic, tangentMode=RCTM_Break, time=0.3, value=1}]}}})
    )) {isSetParameters=true}
  ]
  rendererClasses [&NiagaraSpriteRendererProperties, &NiagaraMeshRendererProperties]
  renderers [
   {rendererClass=&NiagaraSpriteRendererProperties}
   {rendererClass=&NiagaraMeshRendererProperties, rendererIndex=1}
  ]
 }
 Hand_Gloops {
  emitterUpdateScript [
   &EmitterState()
   &SpawnRate(SpawnRate=&FN_FrontEndFloatModifier(Gameplay=2, Front End=0), Use Spawn Probability=false)
   &SpawnBurst_Instantaneous(Spawn Count=1, Use Spawn Probability=false) {enabled=false}
   &SpawnBurst_Instantaneous(Spawn Count=1, Spawn Group=1, Use Spawn Probability=false) {enabled=false, module=SpawnBurst_Instantaneous001}
  ]
  particleSpawnScript [
   &InitializeParticle(
    Color=<rgba 1 1 1 0.47>
    Lifetime=2
    Mesh Scale=&Multiply_VectorByFloat(Vector=<0.8 0.8 1>, Float=&FN_FrontEndFloatModifier(Gameplay=1, Front End=0.5))
    Module.Write Color=true
    Module.Write Lifetime=true
    Module.Write Mass=true
    Module.Write Material Random=true
    Module.Write Position=false
    Module.Write Scale=true
    Module.Write SpriteRotation=false
    Module.Write SpriteSize=true
    Position=&SimulationPosition()
    Sprite Size=<5 5>
   )
   /SetVariables(Particles.PickBone:NiagaraInt32=&ModuloInt(
     A=/Niagara/DynamicInputs/DataInterface/GenerateIntFromCounter(SimpleCounter:NiagaraDataInterfaceSimpleCounter=DataInterface() {propertyValues {GpuSyncMode=None, InitialValue=0}})
     B=2
    )) {isSetParameters=true}
  ]
  particleUpdateScript [
   &ParticleState()
   &SolveForcesAndVelocity(Clamp Velocity=false, Limit Acceleration=false)
   &SkeletalMeshLocation(
    Bone / Socket Index=$Particles.PickBone
    Bone Sampling Mode=::NewEnumerator4
    Inherit Velocity=false
    MaskSpawnGroup=false
    Skeletal Mesh=DataInterface() {propertyValues {
      DefaultMesh=null
      ExcludeBoneName=root
      OptionalComponentClass=null
      PreviewMesh=&M_Med_Soldier_04_Preview
      SkinningMode=SkinOnTheFly
      SoftSourceActor=null
      SourceMode=Default
      UvSetIndex=0
      WholeMeshLOD=-1
      bExcludeBone=false
      bReadDeformedGeometry=true
      bRequireCurrentFrameData=true
      ComponentTags []
      FilteredBones [hand_l,hand_r]
      FilteredSockets []
      MeshUserParameter {parameter {name=None, typeDefHandle {registeredTypeIndex=1}, varData []}}
      SamplingRegions []
     }}
   ) {module=SkeletalMeshLocation001}
   &DynamicMaterialParameters(
    Index 0 Param 1=&RandomRangeFloat(Minimum=0, Maximum=1)
    Index 0 Param 2=5
    Index 0 Param 3=1
    Index 0 Param 4=1
    Module.Param0WriteEnabled=true
    Module.Param1WriteEnabled=true
    Module.Param2WriteEnabled=true
    Module.Param3WriteEnabled=true
   )
  ]
  rendererClasses [&NiagaraSpriteRendererProperties, &NiagaraMeshRendererProperties]
  renderers [
   {rendererClass=&NiagaraSpriteRendererProperties}
   {rendererClass=&NiagaraMeshRendererProperties, rendererIndex=1}
  ]
 }
 Hand_Splats {
  emitterUpdateScript [
   &EmitterState(Life Cycle Mode=::NewEnumerator1, Loop Duration=1, UseLoopDelay=false)
   &SpawnRate(SpawnRate=40, Use Spawn Probability=false)
  ]
  particleSpawnScript [
   &InitializeParticle(
    Color=<rgba 0.6875 0.4889 0.0306 2>
    Lifetime=&RandomRangeFloat(Minimum=0.15, Maximum=0.6)
    Module.Write Color=true
    Module.Write Lifetime=true
    Module.Write Mass=true
    Module.Write Material Random=true
    Module.Write Position=false
    Module.Write Scale=false
    Module.Write SpriteRotation=true
    Module.Write SpriteSize=true
    Position=&SimulationPosition()
    Sprite Rotation=&RandomRangeFloat(Minimum=0, Maximum=360)
    Sprite Size=&Multiply_Vector2(
     A=&RandomRangeVector2D(Minimum=<0.65 0.65>, Maximum=<1 1>)
     B=&Vector2DFromFloat(Value=100)
    )
   )
   &SkeletalMeshLocation(
    Bone Sampling Mode=::NewEnumerator0
    Inherit Velocity=false
    MaskSpawnGroup=false
    Module.Override Randomness=false
    Skeletal Mesh=DataInterface() {propertyValues {
      DefaultMesh=null
      ExcludeBoneName=root
      OptionalComponentClass=null
      PreviewMesh=&M_Med_Soldier_04_Preview
      SkinningMode=SkinOnTheFly
      SoftSourceActor=null
      SourceMode=Default
      UvSetIndex=0
      WholeMeshLOD=-1
      bExcludeBone=false
      bReadDeformedGeometry=true
      bRequireCurrentFrameData=true
      ComponentTags []
      FilteredBones [hand_l,hand_r]
      FilteredSockets []
      MeshUserParameter {parameter {name=None, typeDefHandle {registeredTypeIndex=1}, varData []}}
      SamplingRegions []
     }}
   )
   &SubUVAnimation(End Frame=3, SubUV Animation Mode:ENiagara_SubUVLookupMode=::NewEnumerator2)
  ]
  particleUpdateScript [
   &ParticleState()
   &AddVelocity(Velocity=<0 0 25>)
   &DynamicMaterialParameters(
    Index 0 Param 1=&FloatFromCurve(FloatCurve=DataInterface() {propertyValues {Curve {keys [<at 0 1 RCIM_Cubic>, {arriveTangent=-2.4512, interpMode=RCIM_Cubic, leaveTangent=-2.4512, tangentMode=RCTM_User, time=1, value=0}]}}})
    Index 0 Param 2=&FloatFromCurve(FloatCurve=DataInterface() {propertyValues {Curve {keys [<at 0 0.05 RCIM_Cubic>, {arriveTangent=0.4316, interpMode=RCIM_Cubic, leaveTangent=0.4316, tangentMode=RCTM_User, time=1, value=0.2}]}}})
    Index 0 Param 3=1
    Index 0 Param 4=1
    Module.Param0WriteEnabled=true
    Module.Param1WriteEnabled=true
    Module.Param2WriteEnabled=true
    Module.Param3WriteEnabled=true
   )
   &ScaleSpriteSize(Scale Factor=&Multiply_Vector2(
     A=<1 1>
     B=&Vector2DFromFloat(Value=&FloatFromCurve(FloatCurve=DataInterface() {propertyValues {Curve {keys [<at 0 0.2 RCIM_Cubic>,<at 1 1>]}}}))
    ))
   &SolveForcesAndVelocity(Clamp Velocity=false, Limit Acceleration=false)
   &ScaleColor(
    Scale Alpha=&FloatFromCurve(FloatCurve=DataInterface() {propertyValues {Curve {keys [<at 0 1 RCIM_Cubic>,<at 1 0>]}}})
    ScaleA=true
    ScaleRGB=true
   )
  ]
  rendererClasses [&NiagaraSpriteRendererProperties]
  renderers [{rendererClass=&NiagaraSpriteRendererProperties}]
 }
 Syrup_Ribbons {
  emitterUpdateScript [
   &EmitterState()
   &FN_SpawnRibbon_v2(Number of Ribbons=2, Spawn Mode=::NewEnumerator0) {module=FN_SpawnRibbon_v2001}
   &FN_SpawnRibbon_v2(Number of Ribbons=2, Spawn Mode=::NewEnumerator0) {enabled=false}
   &SpawnRate(SpawnRate=0, Use Spawn Probability=false)
  ]
  particleSpawnScript [
   &FN_InitializeRibbon_v3(
    Enable Skeletal Position=true
    Lifetime=0.7
    Ribbon Width=50
    Sampled Position Coordinate Space:ENiagaraCoordinateSpace=::Local
    Sampled Rotation Coordinate Space:ENiagaraCoordinateSpace=::Local
    Sampled Velocity Coordinate Space:ENiagaraCoordinateSpace=::Local
   ) {module=FN_InitializeRibbon_v3001}
   &AddVelocity(Velocity=&MakeCustomVectorFromBool(
     Bool=$Particles.Locked
     True Vector=<0 0 0>
     False Vector=&FN_FrontEndVector3Modifier(Gameplay=<0 0 410>, Front End=<0 0 350>)
    )) {module=AddVelocity001}
   &DynamicMaterialParameters(
    Index 0 Param 1=&MaskFloatBySpawnGroup(Masked Float=0, Passthrough Float=0.5)
    Module.Param0WriteEnabled=true
    Module.Param1WriteEnabled=false
    Module.Param2WriteEnabled=false
    Module.Param3WriteEnabled=false
   ) {enabled=false, module=DynamicMaterialParameters001}
  ]
  particleUpdateScript [
   &Drag(Drag=0.9, Use Linear Drag=true, Use Rotational Drag=false)
   &InheritVelocity(
    Module.Limit Speed=true
    Velocity Limit=10000
    Velocity Scale=&MakeCustomVectorFromBool(Bool=$Particles.Initial.Locked, True Vector=<0 0 0>, False Vector=<-0.03 -0.03 -0.01>)
    Velocity Source=&TransformVector(Vector=$Engine.Owner.Velocity, Source Space=::World, Destination Space=::Local)
   )
   &AddVelocity(Velocity=&Multiply_VectorByFloat(
     Vector=<0 0 1>
     Float=&FloatFromCurve(FloatCurve=DataInterface() {propertyValues {Curve {keys [<at 0 0 RCIM_Cubic>, {arriveTangent=2.8829, interpMode=RCIM_Cubic, leaveTangent=2.8829, tangentMode=RCTM_User, time=1, value=1}]}}})
    )) {module=AddVelocity002}
   &SolveForcesAndVelocity(Clamp Velocity=false, Limit Acceleration=false)
   &FN_UpdateRibbon(
    Enable Skeletal Position=true
    Initial Color:LinearColor=$Particles.Initial.Color
    Initial RibbonWidth:NiagaraFloat=$Particles.Initial.RibbonWidth
    Sampled Position Coordinate Space:ENiagaraCoordinateSpace=::Local
    Sampled Rotation Coordinate Space:ENiagaraCoordinateSpace=::Local
    Sampled Velocity Coordinate Space:ENiagaraCoordinateSpace=::Local
   )
   &Material(Base Texture Time:NiagaraFloat=&Multiply_Float(A=$Emitter.Age, B=-1))
   &DynamicMaterialParameters(
    Index 0 Param 1=&MaskFloatBySpawnGroup(Masked Float=0, Passthrough Float=0.5)
    Index 0 Param 2=&FloatFromCurve(FloatCurve=DataInterface() {propertyValues {Curve {keys [
         {arriveTangent=-0.3062, interpMode=RCIM_Cubic, leaveTangent=-0.3062, tangentMode=RCTM_User, time=0, value=0.5}
         {arriveTangent=0.0167, interpMode=RCIM_Cubic, leaveTangent=0.0167, tangentMode=RCTM_Break, time=1, value=0}
        ]}}})
    Index 0 Param 3=1
    Index 0 Param 4=1
    Module.Param0WriteEnabled=true
    Module.Param1WriteEnabled=true
    Module.Param2WriteEnabled=true
    Module.Param3WriteEnabled=true
   )
   &ScaleRibbonWidth(Ribbon Width Scale=&FloatFromCurve(
     FloatCurve=DataInterface() {propertyValues {Curve {keys [
         {arriveTangent=-0.0063, interpMode=RCIM_Cubic, leaveTangent=-0.0063, tangentMode=RCTM_User, time=0.0058, value=0.5}
         {arriveTangent=2.456, interpMode=RCIM_Cubic, leaveTangent=2.456, time=0.4538, value=1.5093}
         <at 0.6774 2.1493>
        ]}}}
     Scale Curve=0.6
    ))
   &Color(Color=&ColorFromCurve(ColorCurve=DataInterface() {propertyValues {
       AlphaCurve {keys [<at 0 0 RCIM_Cubic>,<at 0.1231 1>,<at 0.7361 1>]}
       BlueCurve {keys [<at 0 1 RCIM_Cubic>,<at 1 1 RCIM_Cubic>]}
       GreenCurve {keys [<at 0 1 RCIM_Cubic>,<at 1 1 RCIM_Cubic>]}
       RedCurve {keys [<at 0 1 RCIM_Cubic>,<at 1 1 RCIM_Cubic>]}
      }}))
  ]
  rendererClasses [&NiagaraRibbonRendererProperties]
  renderers [{rendererClass=&NiagaraRibbonRendererProperties}]
 }
}
summary {
 systemName=NS_Contrail_Flapjack_Wrangler
 emitters [
  {emitterName=Syrup_Ribbons, rendererClasses [&NiagaraRibbonRendererProperties]}
  {emitterName=Butter_Chips, rendererClasses [&NiagaraSpriteRendererProperties, &NiagaraMeshRendererProperties]}
  {emitterName=Hand_Gloops, rendererClasses [&NiagaraSpriteRendererProperties, &NiagaraMeshRendererProperties]}
  {emitterName=Hand_Splats, rendererClasses [&NiagaraSpriteRendererProperties]}
 ]
 userVariables [
  User.bIsFrontEnd {type=NiagaraBool}
  User.bIsFrontEndPreview {type=NiagaraBool}
  User.PlayerSpeed {type=NiagaraFloat, defaultValue=0, description=""}
  User.PreviousVelocityXY {type=Vector3f, description="", defaultValue {struct=Vector3f, value=<0 0 0>}}
  User.RibbonIDOffset {type=NiagaraInt32}
  User.SkeletalMesh {
   type=NiagaraDataInterfaceSkeletalMesh
   defaultValue {
    struct=NiagaraExt_VariableValue_DataInterface
    value {dataInterface=&self:NiagaraDataInterfaceSkeletalMesh_0}
   }
  }
  User.WaterBodyIndex {type=NiagaraFloat, defaultValue=0, description=""}
  User.WaterDepth {type=NiagaraFloat, defaultValue=0, description=""}
  User.WaterPlaneLocation {type=Vector3f, description="", defaultValue {struct=Vector3f, value=<0 0 0>}}
  User.WaterPlaneNormal {type=Vector3f, description="", defaultValue {struct=Vector3f, value=<0 0 0>}}
  User.WaveTime {type=NiagaraFloat, defaultValue=0, description=""}
 ]
}
